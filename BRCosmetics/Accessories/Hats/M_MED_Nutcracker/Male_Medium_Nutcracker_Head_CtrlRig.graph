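# Control Rig for the male medium Nutcracker hat accessory that drives its dyn_jaw bone from the face's C_jaw bone.
# Each evaluation reads C_jaw's local rotation, converts it to ZYX Euler and multiplies the X angle by 1.5.
# That angle alone becomes a quaternion added in local space to dyn_jaw, a sibling of C_jaw under faceAttach with the same listed transform.
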ControlRigBlueprint /BRCosmetics/Accessories/Hats/M_MED_Nutcracker/Male_Medium_Nutcracker_Head_CtrlRig
graphs [
 RigVMFunctionLibrary {links [], nodes []}
 RigVMModel {nodes [
   n1 = &RigUnit_BeginExecution()
   n2 = &RigVMFunction_MathQuaternionToEuler(RotationOrder=ZYX, Value=&RigUnit_GetBoneTransform(Bone=C_jaw, Space=LocalSpace).Transform.Rotation)
   /Script/ControlRig.RigUnit_ModifyBoneTransforms(
    BoneToModify=((Bone="dyn_jaw",Transform=(Rotation=(X=0,Y=0,Z=0,W=1),Translation=(X=0,Y=0,Z=0),Scale3D=(X=1,Y=1,Z=1))))
    Mode=AdditiveLocal
    Weight=1
    WeightMaximum=1
    WeightMinimum=0
    ExecutePin=@n1.ExecutePin
    BoneToModify.0.Transform.Rotation=&RigVMFunction_MathQuaternionFromEuler(
     Euler=(X=0,Y=0,Z=0)
     Euler.X=&RigVMFunction_MathFloatMul(B=1.5, A=@n2.Result.X) {template="Multiply::Execute(in A,in B,out Result)"}.Result
    ).Result
   ) {is=struct RigVMUnitNode}
  ]}
]
hierarchy [
 root {loc=<0 0 0>, rot=<0 0 0 1>}
 attach {parent=root, loc=<0 0 0>, rot=<0 0 0 1>}
 pelvis {parent=root, loc=<0 -0.7071 93.1191>, rot=<0.0127 0.707 -0.0127 -0.707>}
 spine_01 {parent=pelvis, loc=<6.7494 -0.6388 0>, rot=<0 0 0.034 -0.9994>}
 spine_02 {parent=spine_01, loc=<7.2518 -0.035 0>, rot=<0 0 0.0238 -0.9997>}
 spine_03 {parent=spine_02, loc=<6.0104 0 0>, rot=<0 0 0.0149 -0.9999>}
 spine_04 {parent=spine_03, loc=<8.1024 0 0>, rot=<0 0 -0.0132 -0.9999>}
 spine_05 {parent=spine_04, loc=<10.0146 0 0>, rot=<0 0 -0.0015 -1>}
 clavicle_l {parent=spine_05, loc=<4.3575 -0.1466 -2.3223>, rot=<0.0465 -0.7289 -0.009 -0.6829>}
 upperarm_l {parent=clavicle_l, loc=<15.5111 0 0>, rot=<0.0226 -0.3234 -0.0088 -0.9459>}
 lowerarm_l {parent=upperarm_l, loc=<26.6514 0 0>, rot=<0.0391 -0.0099 -0.2453 -0.9686>}
 hand_l {parent=lowerarm_l, loc=<22.3492 0 0>, rot=<-0.7081 -0.0507 0.0975 -0.6975>}
 index_metacarpal_l {parent=hand_l, loc=<1.6535 0.0183 -2.96>, rot=<0.0158 -0.0124 -0.0173 -0.9996>}
 index_01_l {parent=index_metacarpal_l, loc=<8.3008 0.0357 -0.0524>, rot=<-0.0044 0.0485 0.0924 -0.9945>}
 index_02_l {parent=index_01_l, loc=<3.3587 -0.3003 0.0452>, rot=<-0.002 0.0027 0.3225 -0.9466>}
 index_03_l {parent=index_02_l, loc=<2.5578 -0.123 -0.0252>, rot=<-0.002 0.0067 0.2558 -0.9667>}
 middle_metacarpal_l {parent=hand_l, loc=<2.1856 -0.1725 -0.5984>, rot=<0.006 0.0168 -0.0329 -0.9993>}
 middle_01_l {parent=middle_metacarpal_l, loc=<7.3275 -0.2447 0.0129>, rot=<-0.0208 0.0042 0.0858 -0.9961>}
 middle_02_l {parent=middle_01_l, loc=<4.0123 0.1402 0.1175>, rot=<-0.0218 0.0258 0.3512 -0.9357>}
 middle_03_l {parent=middle_02_l, loc=<2.6077 0.0534 0.1373>, rot=<0.0082 0.0188 0.1708 -0.9851>}
 pinky_metacarpal_l {parent=hand_l, loc=<1.7616 -0.9753 3.3814>, rot=<-0.0397 0.0814 -0.069 -0.9935>}
 pinky_01_l {parent=pinky_metacarpal_l, loc=<7.1709 -0.4831 -0.1094>, rot=<-0.0654 -0.0035 0.2242 -0.9723>}
 pinky_02_l {parent=pinky_01_l, loc=<2.8274 0.0893 -0.1043>, rot=<0.0037 0.0011 0.2854 -0.9584>}
 pinky_03_l {parent=pinky_02_l, loc=<2.2046 0.0308 -0.0078>, rot=<-0.0002 0.0098 0.1445 -0.9895>}
 ring_metacarpal_l {parent=hand_l, loc=<2.2074 -0.3625 1.3638>, rot=<-0.0284 0.0722 -0.0458 -0.9959>}
 ring_01_l {parent=ring_metacarpal_l, loc=<7.0295 -0.6197 -0.1446>, rot=<-0.0489 -0.0252 0.1892 -0.9804>}
 ring_02_l {parent=ring_01_l, loc=<3.6174 0.4646 -0.1027>, rot=<0 0 0.2806 -0.9598>}
 ring_03_l {parent=ring_02_l, loc=<2.4152 0.1141 0.0199>, rot=<0 0 0.1189 -0.9929>}
 thumb_01_l {parent=hand_l, loc=<1.3118 -0.3191 -2.459>, rot=<0.6546 -0.4042 0.0664 -0.6354>}
 thumb_02_l {parent=thumb_01_l, loc=<5.9337 -0.0448 0.0401>, rot=<-0.0037 0.0213 0.298 -0.9543>}
 thumb_03_l {parent=thumb_02_l, loc=<2.2223 0.3682 -0.0907>, rot=<0.0118 -0.013 0.0066 -0.9998>}
 weapon_l {parent=hand_l, loc=<-0.1974 -3.411 0.3811>, rot=<0 0 0 1>}
 lowerarm_twist_01_l {parent=lowerarm_l, loc=<17.8549 0 0>, rot=<0 0 0 -1>}
 lowerarm_twist_02_l {parent=lowerarm_l, loc=<9.6199 0 0>, rot=<0 0 0 -1>}
 upperarm_twist_01_l {parent=upperarm_l, loc=<0.0041 0 0>, rot=<0 0 0 -1>}
 upperarm_twist_02_l {parent=upperarm_l, loc=<13.7941 0 0>, rot=<0 0 0 -1>}
 dyn_strap_clip_fr_l_MSTR {parent=clavicle_l, loc=<12.1796 11.1367 -2.2174>, rot=<0.4186 -0.504 -0.3818 -0.6519>}
 dyn_strap_clip_fr_l {parent=dyn_strap_clip_fr_l_MSTR, loc=<0.2625 -0.0438 0.2723>, rot=<-0.0223 0.2541 -0.0075 -0.9669>}
 dyn_strap_clip_bk_l_MSTR {parent=clavicle_l, loc=<11.6531 -11.3141 -0.439>, rot=<-0.4008 -0.4822 0.4921 -0.6038>}
 dyn_strap_clip_bk_l {parent=dyn_strap_clip_bk_l_MSTR, loc=<0.1948 -0.0817 0.3487>, rot=<-0.024 0.1358 -0.0033 -0.9904>}
 clavicle_r {parent=spine_05, loc=<4.3575 -0.1466 2.3223>, rot=<0.7289 0.0465 -0.6829 0.009>}
 upperarm_r {parent=clavicle_r, loc=<-15.5111 0 0>, rot=<0.0226 -0.3234 -0.0088 -0.9459>}
 lowerarm_r {parent=upperarm_r, loc=<-26.6515 0 0>, rot=<0.0391 -0.0099 -0.2453 -0.9686>}
 hand_r {parent=lowerarm_r, loc=<-22.3493 0 0>, rot=<-0.7081 -0.0507 0.0975 -0.6975>}
 index_metacarpal_r {parent=hand_r, loc=<-1.6534 -0.0181 2.9601>, rot=<0.0158 -0.0124 -0.0173 -0.9996>}
 index_01_r {parent=index_metacarpal_r, loc=<-8.3008 -0.0357 0.0524>, rot=<-0.0044 0.0485 0.0924 -0.9945>}
 index_02_r {parent=index_01_r, loc=<-3.3589 0.3003 -0.0453>, rot=<-0.002 0.0027 0.3225 -0.9466>}
 index_03_r {parent=index_02_r, loc=<-2.5578 0.1231 0.0251>, rot=<-0.002 0.0067 0.2558 -0.9667>}
 middle_metacarpal_r {parent=hand_r, loc=<-2.1856 0.1726 0.5984>, rot=<0.006 0.0168 -0.0329 -0.9993>}
 middle_01_r {parent=middle_metacarpal_r, loc=<-7.3273 0.2447 -0.0129>, rot=<-0.0208 0.0042 0.0858 -0.9961>}
 middle_02_r {parent=middle_01_r, loc=<-4.0142 -0.1411 -0.07>, rot=<-0.0217 0.0259 0.3512 -0.9357>}
 middle_03_r {parent=middle_02_r, loc=<-2.6074 -0.0534 -0.1312>, rot=<0.0083 0.0188 0.1708 -0.9851>}
 pinky_metacarpal_r {parent=hand_r, loc=<-1.7617 0.9753 -3.3814>, rot=<-0.0397 0.0814 -0.069 -0.9935>}
 pinky_01_r {parent=pinky_metacarpal_r, loc=<-7.1707 0.4832 0.1094>, rot=<-0.0654 -0.0035 0.2242 -0.9723>}
 pinky_02_r {parent=pinky_01_r, loc=<-2.8274 -0.0892 0.1043>, rot=<0.0037 0.0011 0.2854 -0.9584>}
 pinky_03_r {parent=pinky_02_r, loc=<-2.2045 -0.031 0.0078>, rot=<-0.0002 0.0098 0.1445 -0.9895>}
 ring_metacarpal_r {parent=hand_r, loc=<-2.2073 0.3627 -1.3638>, rot=<-0.0284 0.0722 -0.0458 -0.9959>}
 ring_01_r {parent=ring_metacarpal_r, loc=<-7.0297 0.6197 0.1446>, rot=<-0.0489 -0.0252 0.1892 -0.9804>}
 ring_02_r {parent=ring_01_r, loc=<-3.6171 -0.4647 0.1027>, rot=<0 0 0.2806 -0.9598>}
 ring_03_r {parent=ring_02_r, loc=<-2.4153 -0.1141 -0.0199>, rot=<0 0 0.1189 -0.9929>}
 thumb_01_r {parent=hand_r, loc=<-1.3119 0.3192 2.459>, rot=<0.6546 -0.4042 0.0664 -0.6354>}
 thumb_02_r {parent=thumb_01_r, loc=<-5.9336 0.0447 -0.0401>, rot=<-0.0037 0.0213 0.298 -0.9543>}
 thumb_03_r {parent=thumb_02_r, loc=<-2.2223 -0.3682 0.0907>, rot=<0.0118 -0.013 0.0066 -0.9998>}
 weapon_r {parent=hand_r, loc=<0.1975 3.4112 -0.3811>, rot=<0 0 0 -1>}
 lowerarm_twist_01_r {parent=lowerarm_r, loc=<-17.855 0 0>, rot=<0 0 0 1>}
 lowerarm_twist_02_r {parent=lowerarm_r, loc=<-9.6199 0 0>, rot=<0 0 0 1>}
 upperarm_twist_01_r {parent=upperarm_r, loc=<-0.0041 0 0>, rot=<0 0 0 -1>}
 upperarm_twist_02_r {parent=upperarm_r, loc=<-13.7942 0 0>, rot=<0 0 0 -1>}
 dyn_armband_knot_MSTR {parent=upperarm_r, loc=<-15.5424 7.3517 2.1107>, rot=<-0.6295 -0.6523 -0.3869 -0.169>}
 dyn_armband_knot {parent=dyn_armband_knot_MSTR, loc=<0.4877 0 0>, rot=<0 0 0 1>}
 dyn_armband_low_01 {parent=dyn_armband_knot, loc=<2.0963 -0.259 -1.4509>, rot=<-0.0037 -0.1522 0.1128 -0.9819>}
 dyn_armband_low_02 {parent=dyn_armband_low_01, loc=<6.7826 0 0>, rot=<0 0 0 1>}
 dyn_armband_hi_01 {parent=dyn_armband_knot, loc=<1.8449 0.5097 0.7988>, rot=<0.0115 0.1786 0.0321 -0.9833>}
 dyn_armband_hi_02 {parent=dyn_armband_hi_01, loc=<5.0757 0 0>, rot=<0 0 0 1>}
 dyn_strap_clip_fr_r_MSTR {parent=clavicle_r, loc=<-12.1796 -11.1367 2.2177>, rot=<-0.3818 0.6519 -0.4186 -0.504>}
 dyn_strap_clip_fr_r {parent=dyn_strap_clip_fr_r_MSTR, loc=<0.2618 0.0439 0.2724>, rot=<0.0223 0.2541 0.0075 -0.9669>}
 dyn_strap_clip_bk_r_MSTR {parent=clavicle_r, loc=<-11.6531 11.3141 0.439>, rot=<-0.4921 -0.6038 -0.4008 0.4822>}
 dyn_strap_clip_bk_r {parent=dyn_strap_clip_bk_r_MSTR, loc=<0.1948 0.0816 0.3487>, rot=<0.024 0.1358 0.0033 -0.9904>}
 neck_01 {parent=spine_05, loc=<11.3138 0.1528 0>, rot=<0 0 -0.2003 -0.9797>}
 neck_02 {parent=neck_01, loc=<4.5627 0 0>, rot=<0 0 0.0504 -0.9987>}
 head {parent=neck_02, loc=<5.3914 0 0>, rot=<0 0 0.115 -0.9934>}
 faceAttach {parent=head, loc=<-0.8266 -0.1832 0>, rot=<0.0044 -0.3827 -0.0044 -0.9239>, scale=<0.89 0.89 0.89>}
 C_brow_mid {parent=faceAttach, loc=<4.7209 11.2084 4.8302>, rot=<-0.0586 -0.317 0.0114 -0.9466>}
 L_brow_mid {parent=faceAttach, loc=<7.639 10.371 3.157>, rot=<0.0603 -0.3852 0.2489 -0.8866>}
 L_brow_outer {parent=faceAttach, loc=<9.1042 8.1218 1.4857>, rot=<0.2349 -0.2525 0.3549 -0.869>}
 L_cheek_inner {parent=faceAttach, loc=<4.2785 8.761 -1.4435>, rot=<0.0454 -0.6262 0.3191 -0.7099>}
 C_lip_upper_mid {parent=faceAttach, loc=<-0.3154 10.7989 -0.2464>, rot=<0.2097 -0.3713 -0.0897 -0.9001>}
 L_lip_upper_outer {parent=faceAttach, loc=<0.8467 10.1076 -1.4454>, rot=<0.1857 -0.3758 0.3015 -0.8564>}
 L_lip_corner {parent=faceAttach, loc=<1.613 9.1457 -2.3124>, rot=<0.0012 -0.375 -0.0189 -0.9269>}
 C_nose_bridge {parent=faceAttach, loc=<2.3602 13.1515 2.4853>, rot=<-0.2002 -0.3676 0.0825 -0.9044>}
 R_brow_mid {parent=faceAttach, loc=<3.0903 10.342 7.7132>, rot=<-0.1347 -0.3556 -0.2231 -0.8976>}
 R_brow_outer {parent=faceAttach, loc=<1.4121 8.0675 9.1444>, rot=<-0.1251 -0.3913 -0.4176 -0.8104>}
 R_cheek_inner {parent=faceAttach, loc=<-1.4992 8.7305 4.3237>, rot=<-0.1979 -0.0545 -0.2598 -0.9436>}
 R_lip_corner {parent=faceAttach, loc=<-2.32 9.1134 1.692>, rot=<0.0127 -0.39 0.008 -0.9207>}
 C_jaw {parent=faceAttach, loc=<0.3123 3.2483 0.3583>, rot=<0.0044 -0.3827 -0.0044 -0.9239>}
 L_lip_lower_outer {parent=C_jaw, loc=<1.6966 6.8705 -0.8455>, rot=<-0.0146 -0.0279 0.3432 -0.9387>}
 C_lip_lower_mid {parent=C_jaw, loc=<0.0184 7.4191 -0.884>, rot=<0 -0.0099 -0.0021 -0.9999>}
 R_lip_lower_outer {parent=C_jaw, loc=<-1.6982 6.9042 -0.8853>, rot=<0.0021 -0.0171 -0.3152 -0.9489>}
 teeth_lower {parent=C_jaw, loc=<0.0169 5.3598 -2.8149>, rot=<0 0 0 1>}
 tongue {parent=C_jaw, loc=<0.0178 1.5245 -1.0399>, rot=<-0.1568 0 0 -0.9876>}
 R_eye {parent=faceAttach, loc=<2.2537 8.4351 6.6885>, rot=<0.3823 -0.0165 0.9223 -0.0548>}
 R_eye_lid_upper_mid {parent=R_eye, loc=<-0.0465 -1.169 0.2722>, rot=<0.0559 -0.0011 0.0197 -0.9982>, scale=<1 1.0001 0.9999>}
 R_eye_lid_lower_mid {parent=R_eye, loc=<-0.0141 -0.9451 -0.7006>, rot=<0.035 -0.0006 0.0197 -0.9992>}
 L_eye {parent=faceAttach, loc=<6.6314 8.4647 2.3108>, rot=<0.3819 0.0253 0.9227 0.046>}
 L_eye_lid_upper_mid {parent=L_eye, loc=<0.0251 -1.1623 0.2388>, rot=<0.0523 0 0 -0.9986>}
 L_eye_lid_lower_mid {parent=L_eye, loc=<0.0223 -0.923 -0.7488>, rot=<0.0329 0 0 -0.9995>}
 teeth_upper {parent=faceAttach, loc=<0.0954 9.3381 0.1585>, rot=<0.0044 -0.3827 -0.0044 -0.9239>}
 R_lip_upper_outer {parent=faceAttach, loc=<-1.497 10.1359 0.962>, rot=<-0.0848 -0.2926 -0.3215 -0.8966>}
 dyn_jaw {parent=faceAttach, loc=<0.3123 3.2483 0.3583>, rot=<0.0044 -0.3827 -0.0044 -0.9239>}
 dyn_collar_fr_r_MSTR {parent=spine_05, loc=<2.2225 9.5483 5.7859>, rot=<0.8048 0.4894 0.2488 0.2255>}
 dyn_collar_fr_r_01 {parent=dyn_collar_fr_r_MSTR, loc=<0.8292 0 0>, rot=<0 0 0 -1>}
 dyn_collar_fr_r_02 {parent=dyn_collar_fr_r_01, loc=<5.5147 -0.1547 0.7264>, rot=<0 0 0 -1>}
 dyn_collar_back_MSTR {parent=spine_05, loc=<13.994 -5.0394 0>, rot=<-0.7032 0.0747 0.0747 -0.7032>}
 dyn_collar_back {parent=dyn_collar_back_MSTR, loc=<0.5899 0 0>, rot=<0 0 0 -1>}
 dyn_collar_bk_l_MSTR {parent=spine_05, loc=<13.8632 -3.2125 -6.9158>, rot=<0.9052 0.019 -0.1727 0.3878>}
 dyn_collar_bk_l {parent=dyn_collar_bk_l_MSTR, loc=<0.5899 0 0>, rot=<0 0 0 -1>}
 dyn_collar_l_MSTR {parent=spine_05, loc=<10.7339 3.1427 -9.8401>, rot=<0.9457 0.1944 -0.258 0.0354>}
 dyn_collar_l {parent=dyn_collar_l_MSTR, loc=<0.8292 0 0>, rot=<0 0 0 -1>}
 dyn_collar_fr_l_MSTR {parent=spine_05, loc=<4.2984 8.9751 -8.9508>, rot=<0.9235 0.3447 -0.1532 -0.07>}
 dyn_collar_fr_l_01 {parent=dyn_collar_fr_l_MSTR, loc=<0.8292 0 0>, rot=<0 0 0 -1>}
 dyn_collar_fr_l_02 {parent=dyn_collar_fr_l_01, loc=<5.663 -0.4923 0.7959>, rot=<-0.0479 0.1389 0.5158 -0.844>}
 dyn_collar_fr_l_03 {parent=dyn_collar_fr_l_02, loc=<2.6506 0 0>, rot=<0 0 0 -1>}
 dyn_backpack {parent=spine_05, loc=<0.9371 -11.6559 0>, rot=<0 0 0 -1>}
 dyn_collar_bk_r_MSTR {parent=spine_05, loc=<13.863 -3.2125 6.9158>, rot=<-0.3878 0.1727 -0.019 -0.9052>}
 dyn_collar_bk_r {parent=dyn_collar_bk_r_MSTR, loc=<0.5901 0 -0.0001>, rot=<0 0 0 -1>}
 dyn_collar_r_MSTR {parent=spine_05, loc=<10.7337 3.1427 9.8401>, rot=<-0.0354 0.258 -0.1944 -0.9457>}
 dyn_collar_r {parent=dyn_collar_r_MSTR, loc=<0.829 0.0001 0.0001>, rot=<0 0 0 -1>}
 dyn_chest_zipper_r_MSTR {parent=spine_04, loc=<-0.1718 9.6375 11.4356>, rot=<-0.0288 0.3813 0.9239 -0.012>}
 dyn_chest_zipper_r {parent=dyn_chest_zipper_r_MSTR, loc=<0 0 0.7154>, rot=<0 0 0 -1>}
 dyn_collar_zipper_MSTR {parent=spine_04, loc=<2.7589 14.5377 -4.3214>, rot=<-0.0276 -0.8063 -0.5838 -0.0911>}
 dyn_collar_zipper {parent=dyn_collar_zipper_MSTR, loc=<0 0 0.5712>, rot=<0 0 0 -1>}
 thigh_l {parent=pelvis, loc=<-2.5693 -0.0109 -7.7501>, rot=<0.0012 0.0572 0.0155 -0.9982>}
 calf_l {parent=thigh_l, loc=<-39.0485 0 0>, rot=<0.0006 -0.0161 -0.0391 -0.9991>}
 calf_twist_01_l {parent=calf_l, loc=<-35.8131 -0.0967 0.0397>, rot=<0.0237 0.0095 -0.0035 -0.9997>}
 calf_twist_02_l {parent=calf_l, loc=<-18.8064 -0.0967 0.0397>, rot=<0.0237 0.0095 -0.0035 -0.9997>}
 dyn_boot_zipper_out_l_MSTR {parent=calf_twist_02_l, loc=<-7.0145 4.2622 -4.8369>, rot=<-0.0867 -0.9366 -0.335 -0.0554>}
 dyn_boot_zipper_out_l {parent=dyn_boot_zipper_out_l_MSTR, loc=<0.8318 0 0>, rot=<0 0 0 -1>}
 dyn_boot_zipper_in_l_MSTR {parent=calf_twist_02_l, loc=<-6.5767 3.9294 2.8031>, rot=<0.2683 0.4325 0.8556 0.094>}
 dyn_boot_zipper_in_l {parent=dyn_boot_zipper_in_l_MSTR, loc=<0.8318 0 0>, rot=<0 0 0 -1>}
 foot_l {parent=calf_l, loc=<-42.561 0.1077 0.4292>, rot=<0.0418 -0.0455 0.0343 -0.9975>}
 ball_l {parent=foot_l, loc=<-6.1598 13.8264 0.6806>, rot=<0.0001 -0.0001 -0.7063 -0.7079>}
 dyn_kneepad_l {parent=calf_l, loc=<0.2806 1.6311 0.1615>, rot=<0 0 0 1>}
 thigh_twist_01_l {parent=thigh_l, loc=<-21.6303 0 0>, rot=<0.0013 0 0 -1>}
 dyn_hip_pouch_fr_l {parent=thigh_twist_01_l, loc=<8.7863 6.351 -5.7183>, rot=<-0.0089 -0.8948 -0.4453 -0.0303>}
 thigh_r {parent=pelvis, loc=<-2.5691 -0.0109 7.7501>, rot=<0.0572 -0.0012 0.9982 0.0155>}
 calf_r {parent=thigh_r, loc=<39.0487 0 0>, rot=<0.0006 -0.0161 -0.0391 -0.9991>}
 calf_twist_01_r {parent=calf_r, loc=<35.8132 0.0967 -0.0397>, rot=<0.0237 0.0095 -0.0035 -0.9997>}
 calf_twist_02_r {parent=calf_r, loc=<18.8064 0.0967 -0.0397>, rot=<0.0237 0.0095 -0.0035 -0.9997>}
 dyn_boot_zipper_in_r_MSTR {parent=calf_twist_02_r, loc=<6.5766 -3.9294 -2.8031>, rot=<0.8556 -0.094 -0.2683 0.4325>}
 dyn_boot_zipper_in_r {parent=dyn_boot_zipper_in_r_MSTR, loc=<0.8318 0 0>, rot=<0 0 0 -1>}
 dyn_boot_zipper_out_r_MSTR {parent=calf_twist_02_r, loc=<7.0144 -4.2622 4.8369>, rot=<-0.335 0.0554 0.0867 -0.9366>}
 dyn_boot_zipper_out_r {parent=dyn_boot_zipper_out_r_MSTR, loc=<0.8317 0 0>, rot=<0 0 0 -1>}
 foot_r {parent=calf_r, loc=<42.5612 -0.1077 -0.4292>, rot=<0.0418 -0.0455 0.0343 -0.9975>}
 ball_r {parent=foot_r, loc=<6.1598 -13.8264 -0.6806>, rot=<0.0001 -0.0001 -0.7063 -0.7079>}
 dyn_kneepad_r {parent=calf_r, loc=<-0.2806 -1.6311 -0.1615>, rot=<0 -1 0 0>}
 thigh_twist_01_r {parent=thigh_r, loc=<21.6303 0 0>, rot=<0.0013 0 0 -1>}
 dyn_hip_pouch_fr_r {parent=thigh_twist_01_r, loc=<-13.7598 -4.6583 8.1361>, rot=<-0.2171 -0.0194 -0.0237 -0.9757>}
 dyn_hip_pouch_bk_r {parent=thigh_twist_01_r, loc=<-14.2207 1.6071 8.341>, rot=<0.0608 -0.0028 -0.0452 -0.9971>}
 dyn_belt_back_pouch {parent=pelvis, loc=<13.428 -11.1495 8.3453>, rot=<0.0294 -0.4704 0.88 -0.058>}
 dyn_belt_bullet_01 {parent=pelvis, loc=<12.0068 3.0615 14.0419>, rot=<-0.0181 0.2568 0.9643 -0.0625>}
 dyn_belt_bullet_02 {parent=pelvis, loc=<11.5551 6.5523 12.2223>, rot=<-0.0208 0.2977 0.9525 -0.0616>}
 dyn_belt_bullet_03 {parent=pelvis, loc=<11.162 9.6073 9.8765>, rot=<-0.023 0.3317 0.9411 -0.0609>}
 pelvis_l {parent=pelvis, loc=<0.1622 0 -0.1>, rot=<0 0 0 -1>}
 pelvis_r {parent=pelvis, loc=<0.1622 0 0.1>, rot=<0 0 0 -1>}
 pelvis_side_l {parent=pelvis, loc=<0.024 0 -0.1>, rot=<0 0 0 -1>}
 pelvis_mid_l {parent=pelvis, loc=<0.1163 0 -0.1>, rot=<0 0 0 -1>}
 pelvis_side_r {parent=pelvis, loc=<0.024 0 0.1>, rot=<0 0 0 -1>}
 pelvis_mid_r {parent=pelvis, loc=<0.1163 0 0.1>, rot=<0 0 0 -1>}
 ik_foot_root {parent=root, loc=<0 0 0>, rot=<0 0 0 1>}
 ik_foot_l {parent=ik_foot_root, loc=<15.2608 -4.4176 9.4433>, rot=<-0.0357 -0.7045 -0.0284 0.7082>}
 ik_foot_r {parent=ik_foot_root, loc=<-15.2609 -4.4176 9.4431>, rot=<-0.7082 -0.0284 0.7045 -0.0357>}
 ik_hand_root {parent=root, loc=<0 0 0>, rot=<0 0 0 1>}
 ik_hand_gun {parent=ik_hand_root, loc=<-33.8029 3.8283 90.5694>, rot=<-0.4834 -0.3223 -0.5887 0.5621>}
 ik_hand_l {parent=ik_hand_gun, loc=<6.7067 65.8041 13.978>, rot=<-0.0992 -0.9733 -0.2068 0>}
 ik_hand_r {parent=ik_hand_gun, loc=<0 0 0>, rot=<0 0 0 -1>}
]
